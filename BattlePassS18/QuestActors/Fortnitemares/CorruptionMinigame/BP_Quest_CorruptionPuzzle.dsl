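; Fortnitemares quest puzzle actor in which players bury three pairs of corrupted shards by damaging them.
; An eligible player entering the perimeter raises the outer shards; burying one raises its inner partner, with timeouts undoing progress.
; Burying all shards applies a gameplay effect and quest progress to participants, then resets, as does the last participant leaving.
; The server tracks state in a struct map; clients animate shards through one timeline per pair.
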(event EventBeginPlay
  ;[1 1 1] Should not be visible to anyone until they enter the puzzle perimeter
  (|CallParentFunction)
  (Default|ResetShardMeshes))

(event OnComponentBeginOverlap(PuzzlePerimeterCollision)
  (params (OverlappedComponent :type Primitive%20Component%20Object%20Reference) (OtherActor :type Actor%20Object%20Reference) (OtherComp :type Primitive%20Component%20Object%20Reference) (OtherBodyIndex :type Integer) (bFromSweep :type Boolean) (SweepResult :type Hit%20Result%20Structure%20%28by%20ref%29))
  ;[1 1 1] Any actor that overlaps is checked. If they aren't a potential valid participant, we add them to the checked actors. If they're just missing the quest, we do nothing. If they have the quest, we add them to CheckedActors and ParticipantPlayers.
  (Networking|SwitchHasAuthority
    (:Authority
      (if (not (Utilities|Set|ContainsItem (Variables|Data|GetCheckedActors) OtherActor))
        (bind _asfort_player_pawn (Utilities|Casting|CastToFortPlayerPawn OtherActor)
          (:then
            (Utilities|IsValid (Fortnite|GetFortPlayerControllerfromActor _asfort_player_pawn)
              (:Is%20Valid
                (Utilities|IsValid (Quests|GetQuestManager (Fortnite|GetFortPlayerControllerfromActor _asfort_player_pawn) "Invalid")
                  (:Is%20Valid
                    (bind _cancompletequest (Class|AthenaQuestFunctionLibrary|AthenaCheckCreativeMode (Game|GetGameState)))
                    (if _cancompletequest
                      (bind _bdidactorhavetags (Class|AthenaQuestFunctionLibrary|AthenaCheckActorHasTags OtherActor (Variables|Data|GetDiscoveryTags) true))
                      (if (and (not _bdidactorhavetags) (or (Quest|IsQuestinProgress (Quests|GetQuestManager (Fortnite|GetFortPlayerControllerfromActor _asfort_player_pawn) "Invalid") (Variables|BPVars|Setup|GetQuestID)) (Variables|BPVars|Setup|GetTest)))
                        (Utilities|Set|Add (Variables|Data|GetParticipantPlayers) (Fortnite|GetFortPlayerControllerfromActor _asfort_player_pawn))
                        (CallFunction|InitializePuzzle))
                      (else
                        _))
                    (Utilities|Set|Add (Variables|Data|GetCheckedActors) OtherActor))
                  (:Is%20Not%20Valid
                    _)))
              (:Is%20Not%20Valid
                _)))
          (:CastFailed
            _))))))

(event Custom|InitializePuzzle
  ;[1 1 1] Check if we're on the right client, then set all inner shards to be buried and have all outer shards raise.
  (if (not (Variables|Data|GetIsPuzzleActive))
    (Variables|Data|SetIsPuzzleActive true)
    (for _array_element (Variables|Data|GetShardsArray)
      (if (not (CallFunction|IsInnerShard _array_element))
        (CallFunction|RaiseShard _array_element)))))

(event Custom|InitializeBuried
  (params (Shard%20Mesh :type Static%20Mesh%20Component%20Object%20Reference) (Ground%20Mesh :type Static%20Mesh%20Component%20Object%20Reference))
  ;[1 1 1] Child shards start buried and don't need to be reburied cosmetically.
  (Transformation|SetRelativeLocationAndRotation Shard%20Mesh "0, 0,-70.000000" (Variables|/Script/Engine.SceneComponent|GetRelativeRotation Shard%20Mesh) false false)
  (Transformation|SetRelativeScale3D Ground%20Mesh "0.050000,0.050000,0.050000")
  (Transformation|SetRelativeLocation Ground%20Mesh "0, 0,-10.000000" false false))

(event EventOnDamageServer
  (params (Damage :type Float%20%28single-precision%29) (DamageTags :type Gameplay%20Tag%20Container%20Structure%20%28by%20ref%29) (Momentum :type Vector :default "0, 0, 0") (HitInfo :type Hit%20Result%20Structure%20%28by%20ref%29) (InstigatedBy :type Controller%20Object%20Reference) (DamageCauser :type Actor%20Object%20Reference) (EffectContext :type Gameplay%20Effect%20Context%20Handle%20Structure))
  (bind (_bblockinghit _binitialoverlap _time _distance _location _impactpoint _normal _impactnormal _physmat _hitactor _hitcomponent) (Collision|BreakHitResult HitInfo))
  ;[1 1 1] On hit, we bury the shard
  (Networking|SwitchHasAuthority
    (:Authority
      (Utilities|Casting|CastToFortPlayerController InstigatedBy)
      (if (GameplayTags|HasAnyTags DamageTags (Variables|BPVars|Setup|GetDamageTags) true)
        (CallFunction|BuryShard (Transformation|GetAttachParent _hitcomponent) true)))))

(event Custom|ShardCostmetics
  (params (Shard :type Scene%20Component%20Object%20Reference))
  ;[1 1 1] VFX that play on both raise or bury
  (Niagara|SpawnSystemAttached (Variables|BPVars|Art|GetCrystalSpawn) Shard "None" "0, 0,5.000000" "0, 0, 0" "KeepRelativeOffset" true))

(event Custom|BuriedCosmetics
  (params (BuriedByPlayer :type Boolean) (Shard :type Scene%20Component%20Object%20Reference))
  ;[1 1 1] Burst VFX on bury and toggling off ambient VFX and audio
  (CallFunction|ToggleCosmeticComponents false Shard)
  (if BuriedByPlayer
    (Niagara|SpawnSystemAttached (Variables|BPVars|Art|GetCrystalImpact) Shard "None" "0, 0,30.000000" "0, 0, 0" "KeepRelativeOffset" true)
    (Niagara|SpawnSystemAttached (Variables|BPVars|Art|GetCrystalSuccess) Shard "None" "0, 0,30.000000" "0, 0, 0" "KeepRelativeOffset" true)))

(event Custom|BuryShard
  (params (Shard :type Scene%20Component%20Object%20Reference) (BuriedByPlayer :type Boolean))
  ;[1 1 1] Bury or raise the shard (cosmetics), then send the shard to be moved by a timeline.
  (CallFunction|ToggleCollision false Shard)
  (Networking|SwitchHasAuthority
    (:Authority
      (CallFunction|SetIsRaised Shard))
    (:Remote
      (CallFunction|BuriedCosmetics BuriedByPlayer Shard)
      (CallFunction|ShardCostmetics Shard)
      (Audio|SpawnSoundAttached (Variables|BPVars|Art|GetBurySound) Shard)))
  (switch int (Default|GetShardIndex Shard)
    (:0
      (CallFunction|ShardPicker1 Shard true BuriedByPlayer))
    (:1
      (CallFunction|ShardPicker2 Shard true BuriedByPlayer))
    (:2
      (CallFunction|ShardPicker3 Shard true BuriedByPlayer))))

(event Custom|RaiseShard
  (params (Shard :type Scene%20Component%20Object%20Reference))
  ;[1 1 1] Bury or raise the shard (cosmetics), then send the shard to be moved by a timeline.
  (Rendering|SetVisibility Shard true true)
  (Networking|SwitchHasAuthority
    (:Authority
      (CallFunction|SetIsRaised Shard true))
    (:Remote
      (CallFunction|ShardCostmetics Shard)
      (Audio|SpawnSoundAttached (Variables|BPVars|Art|GetRaiseSound) Shard)))
  (switch int (Default|GetShardIndex Shard)
    (:0
      (CallFunction|ShardPicker1 Shard))
    (:1
      (CallFunction|ShardPicker2 Shard))
    (:2
      (CallFunction|ShardPicker3 Shard))))

(event Custom|OnTimelineFinished
  (params (Shard :type Scene%20Component%20Object%20Reference) (IsBeingBuried :type Boolean) (BuriedByPlayer :type Boolean))
  (if IsBeingBuried
    (Networking|SwitchHasAuthority
      (:Authority
        (CallFunction|HandleBuried Shard BuriedByPlayer))
      (:Remote
        (Rendering|SetVisibility Shard false true)))
    (else
      (CallFunction|ToggleCollision true Shard)
      (Networking|SwitchHasAuthority
        (:Authority
          (CallFunction|HandleRaised Shard))
        (:Remote
          (CallFunction|ToggleCosmeticComponents true Shard))))))

(event Custom|ShardPicker1
  (params (Shard :type Scene%20Component%20Object%20Reference) (IsBeingBuried :type Boolean) (BuriedByPlayer :type Boolean))
  ;[1 1 1] Cosmetically adjust the shard and ground to raise or lower. Each pair of shards needs their own timeline.
  (if IsBeingBuried
    (bind (_direction _shardheight _shardrotation _crackscale _crackheight) (Timeline|ShardTimeline_0 PlayFromStart)
      (:Update
        (CallFunction|TimelineUpdate _shardheight _shardrotation _crackscale _crackheight Shard))
      (:Finished
        (CallFunction|OnTimelineFinished Shard IsBeingBuried BuriedByPlayer)))
    (else
      (Timeline|ShardTimeline_0 ReverseFromEnd))))

(event Custom|ShardPicker2
  (params (Shard :type Scene%20Component%20Object%20Reference) (IsBeingBuried :type Boolean) (BuriedByPlayer :type Boolean))
  ;[1 1 1] Cosmetically adjust the shard and ground to raise or lower. Each pair of shards needs their own timeline.
  (if IsBeingBuried
    (bind (_direction _shardheight _shardrotation _crackscale _crackheight) (Timeline|ShardTimeline_1 PlayFromStart)
      (:Update
        (CallFunction|TimelineUpdate _shardheight _shardrotation _crackscale _crackheight Shard))
      (:Finished
        (CallFunction|OnTimelineFinished Shard IsBeingBuried BuriedByPlayer)))
    (else
      (Timeline|ShardTimeline_1 ReverseFromEnd))))

(event Custom|ShardPicker3
  (params (Shard :type Scene%20Component%20Object%20Reference) (IsBeingBuried :type Boolean) (BuriedByPlayer :type Boolean))
  ;[1 1 1] Cosmetically adjust the shard and ground to raise or lower. Each pair of shards needs their own timeline.
  (if IsBeingBuried
    (bind (_direction _shardheight _shardrotation _crackscale _crackheight) (Timeline|ShardTimeline_2 PlayFromStart)
      (:Update
        (CallFunction|TimelineUpdate _shardheight _shardrotation _crackscale _crackheight Shard))
      (:Finished
        (CallFunction|OnTimelineFinished Shard IsBeingBuried BuriedByPlayer)))
    (else
      (Timeline|ShardTimeline_2 ReverseFromEnd))))

(event Custom|HandleBuried
  (params (Shard :type Scene%20Component%20Object%20Reference) (BuriedByPlayer :type Boolean))
  (if (CallFunction|IsInnerShard Shard)
    (if BuriedByPlayer
      (bind _puzzlecompleted (CallFunction|CheckAllBuried))
      (if (not _puzzlecompleted)
        (switch int (Default|GetShardIndex Shard)
          (:0
            (Utilities|Time|ClearandInvalidateTimerbyHandle (Variables|Data|GetTimer_Shard1))
            (bind _returnvalue (Utilities|Time|SetTimerbyEvent (EventDispatchers|CreateEvent :!event Custom|TimeoutRaiseShard1) (Variables|BPVars|Setup|GetShardTimeout)))
            (Variables|Data|SetTimer_Shard1 _returnvalue))
          (:1
            (Utilities|Time|ClearandInvalidateTimerbyHandle (Variables|Data|GetTimer_Shard2))
            (bind _returnvalue_2 (Utilities|Time|SetTimerbyEvent (EventDispatchers|CreateEvent :!event Custom|TimeoutRaiseShard2) (Variables|BPVars|Setup|GetShardTimeout)))
            (Variables|Data|SetTimer_Shard2 _returnvalue_2))
          (:2
            (Utilities|Time|ClearandInvalidateTimerbyHandle (Variables|Data|GetTimer_Shard3))
            (bind _returnvalue_3 (Utilities|Time|SetTimerbyEvent (EventDispatchers|CreateEvent :!event Custom|TimeoutRaiseShard3) (Variables|BPVars|Setup|GetShardTimeout)))
            (Variables|Data|SetTimer_Shard3 _returnvalue_3))))
      (else
        (CallFunction|RaiseShard (CallFunction|GetParentShard Shard))))
    (else
      (CallFunction|RaiseShard (CallFunction|GetChildShard Shard)))))

(event Custom|TimeoutRaiseShard1
  (CallFunction|RaiseShard (Variables|Default|GetShard_Inner_1)))

(event Custom|TimeoutRaiseShard2
  (CallFunction|RaiseShard (Variables|Default|GetShard_Inner_2)))

(event Custom|TimeoutRaiseShard3
  (CallFunction|RaiseShard (Variables|Default|GetShard_Inner_3)))

(event Custom|HandleRaised
  (params (Shard :type Scene%20Component%20Object%20Reference))
  (if (CallFunction|IsInnerShard Shard)
    (switch int (Default|GetShardIndex Shard)
      (:0
        (Utilities|Time|ClearandInvalidateTimerbyHandle (Variables|Data|GetTimer_Shard1))
        (bind _returnvalue (Utilities|Time|SetTimerbyEvent (EventDispatchers|CreateEvent :!event Custom|ReburyInnerShard1) (Variables|BPVars|Setup|GetShardTimeout)))
        (Variables|Data|SetTimer_Shard1 _returnvalue))
      (:1
        (Utilities|Time|ClearandInvalidateTimerbyHandle (Variables|Data|GetTimer_Shard2))
        (bind _returnvalue_2 (Utilities|Time|SetTimerbyEvent (EventDispatchers|CreateEvent :!event Custom|ReburyInnerShard2) (Variables|BPVars|Setup|GetShardTimeout)))
        (Variables|Data|SetTimer_Shard2 _returnvalue_2))
      (:2
        (Utilities|Time|ClearandInvalidateTimerbyHandle (Variables|Data|GetTimer_Shard3))
        (bind _returnvalue_3 (Utilities|Time|SetTimerbyEvent (EventDispatchers|CreateEvent :!event Custom|ReburyInnerShard3) (Variables|BPVars|Setup|GetShardTimeout)))
        (Variables|Data|SetTimer_Shard3 _returnvalue_3)))))

(event Custom|ReburyInnerShard1
  (CallFunction|BuryShard (Variables|Default|GetShard_Inner_1)))

(event Custom|ReburyInnerShard2
  (CallFunction|BuryShard (Variables|Default|GetShard_Inner_2)))

(event Custom|ReburyInnerShard3
  (CallFunction|BuryShard (Variables|Default|GetShard_Inner_3)))

(event OnComponentEndOverlap(PuzzlePerimeterCollision)
  (params (OverlappedComponent :type Primitive%20Component%20Object%20Reference) (OtherActor :type Actor%20Object%20Reference) (OtherComp :type Primitive%20Component%20Object%20Reference) (OtherBodyIndex :type Integer))
  (bind _asfort_player_pawn (Utilities|Casting|CastToFortPlayerPawn OtherActor))
  (Utilities|Set|Remove (Variables|Data|GetCheckedActors) _asfort_player_pawn)
  (Utilities|IsValid (Fortnite|GetFortPlayerControllerfromActor _asfort_player_pawn)
    (:Is%20Valid
      (Utilities|Set|Remove (Variables|Data|GetParticipantPlayers) (Fortnite|GetFortPlayerControllerfromActor _asfort_player_pawn))
      (if (== (Utilities|Set|Length (Variables|Data|GetParticipantPlayers)) 0)
        (CallFunction|ResetPuzzle)))))

(event Custom|ResetPuzzle
  (Utilities|Set|Clear (Variables|Data|GetParticipantPlayers))
  (Default|ResetShardsMap)
  (Default|ResetShardMeshes)
  (Utilities|Time|ClearandInvalidateTimerbyHandle (Variables|Data|GetTimer_Shard1))
  (Utilities|Time|ClearandInvalidateTimerbyHandle (Variables|Data|GetTimer_Shard2))
  (Utilities|Time|ClearandInvalidateTimerbyHandle (Variables|Data|GetTimer_Shard3))
  (Variables|Data|SetIsPuzzleActive false))

(fn AddToShardsMap
  (params (SceneComponent :type Scene%20Component%20Object%20Reference) (ShardMesh :type Static%20Mesh%20Component%20Object%20Reference) (GroundMesh :type Static%20Mesh%20Component%20Object%20Reference) (Collision :type Capsule%20Collision%20Object%20Reference) (VFX :type Niagara%20Particle%20System%20Component%20Object%20Reference) (Audio :type Audio%20Component%20Object%20Reference) (IsRaised :type Boolean))
  (Utilities|Map|Add (Variables|Data|GetShardsMap) SceneComponent (Utilities|Struct|MakeCorruptionPuzzleShard ShardMesh GroundMesh Collision VFX Audio IsRaised)))

(fn CheckAllBuried
  (Variables|Data|SetShouldApplyGE true)
  (bind _values (Utilities|Map|Values (Variables|Data|GetShardsMap)))
  (bind _array_element (Utilities|Array|ForEachLoopwithBreak _values)
    (:LoopBody
      (if (.IsRaised_30_B8258F67497498D9777F1DBB501B5CB6 (Utilities|Struct|BreakCorruptionPuzzleShard _array_element))
        (Variables|Data|SetShouldApplyGE false)))
    (:Completed
      _))
  (if (Variables|Data|GetShouldApplyGE)
    (bind _result (Utilities|Set|ToArray (Variables|Data|GetParticipantPlayers)))
    (for _array_element_2 _result
      (Environment|ApplyGlobalEnvironmentGameplayEffecttoActor _array_element_2 (Variables|Data|GetQuestGE))
      (CallFunction|GrantQuestProgress _array_element_2))
    (CallFunction|ResetPuzzle)
    (return true)
    (else
      (return (not (Variables|Data|GetIsPuzzleActive))))))

(fn GetChildShard
  (params (ParentShard :type Scene%20Component%20Object%20Reference%20%28by%20ref%29))
  (return (Utilities|Array|Get(acopy) (Variables|Data|GetShardsArray) (- (Utilities|Array|FindItem (Variables|Data|GetShardsArray) ParentShard) 3))))

(fn GetParentShard
  (params (ParentShard :type Scene%20Component%20Object%20Reference%20%28by%20ref%29))
  (return (Utilities|Array|Get(acopy) (Variables|Data|GetShardsArray) (+ (Utilities|Array|FindItem (Variables|Data|GetShardsArray) ParentShard) 3))))

(fn GetShardIndex
  (params (InputShard :type Scene%20Component%20Object%20Reference))
  (bind _shardsarray (Variables|Data|GetShardsArray))
  (if (Utilities|Array|ContainsItem _shardsarray InputShard)
    (return (Math|Integer|%25%28Integer%29 (Utilities|Array|FindItem _shardsarray InputShard) 3))
    (else
      (return -1))))

(fn GrantQuestProgress
  (params (Participant :type Fort%20Player%20Controller%20Object%20Reference))
  (bind _returnvalue (Quests|GetQuestManager Participant "Invalid"))
  (Utilities|IsValid Participant
    (:Is%20Valid
      (Utilities|IsValid _returnvalue
        (:Is%20Valid
          (Quest|SendComplexCustomStatEvent _returnvalue self "(GameplayTags=)" (GameplayTags|GetOwnedGameplayTags self :!class /Script/GameplayTags.GameplayTagAssetInterface) 1))))))

(fn IsInnerShard
  (params (ItemToFind :type Scene%20Component%20Object%20Reference%20%28by%20ref%29))
  (return (< (Utilities|Array|FindItem (Variables|Data|GetShardsArray) ItemToFind) 3)))

(fn ResetShardMeshes
  (bind _shardsmap (Variables|Data|GetShardsMap))
  (bind _keys (Utilities|Map|Keys (Variables|Data|GetShardsMap)))
  (for _array_element _keys
    (if (.ReturnValue (Utilities|Map|Find _shardsmap _array_element))
      (CallFunction|InitializeBuried (Utilities|Struct|BreakCorruptionPuzzleShard (Utilities|Map|Find _shardsmap _array_element)) (.GroundMesh_22_FEF31656488A70C8ACD70EB8EDF69B97 (Utilities|Struct|BreakCorruptionPuzzleShard (Utilities|Map|Find _shardsmap _array_element))))
      (CallFunction|ToggleCosmeticComponents false _array_element)
      (Rendering|SetVisibility _array_element false true)
      (CallFunction|ToggleCollision false _array_element))))

(fn ResetShardsMap
  (Utilities|Map|Clear (Variables|Data|GetShardsMap))
  (CallFunction|AddtoShardsMap (Variables|Default|GetShard_Inner_1) (Variables|Default|GetShardMesh_Inner_1) (Variables|Default|GetGroundMesh_Inner_1) (Variables|Default|GetCollision_Inner_1) (Variables|Default|GetVFX_Inner_1) (Variables|Default|GetAudio_Inner_1))
  (CallFunction|AddtoShardsMap (Variables|Default|GetShard_Inner_2) (Variables|Default|GetShardMesh_Inner_2) (Variables|Default|GetGroundMesh_Inner_2) (Variables|Default|GetCollision_Inner_2) (Variables|Default|GetVFX_Inner_2) (Variables|Default|GetAudio_Inner_2))
  (CallFunction|AddtoShardsMap (Variables|Default|GetShard_Inner_3) (Variables|Default|GetShardMesh_Inner_3) (Variables|Default|GetGroundMesh_Inner_3) (Variables|Default|GetCollision_Inner_3) (Variables|Default|GetVFX_Inner_3) (Variables|Default|GetAudio_Inner_3))
  (CallFunction|AddtoShardsMap (Variables|Default|GetShard_Outer_1) (Variables|Default|GetShardMesh_Outer_1) (Variables|Default|GetGroundMesh_Outer_1) (Variables|Default|GetCollision_Outer_1) (Variables|Default|GetVFX_Outer_1) (Variables|Default|GetAudio_Outer_1) true)
  (CallFunction|AddtoShardsMap (Variables|Default|GetShard_Outer_2) (Variables|Default|GetShardMesh_Outer_2) (Variables|Default|GetGroundMesh_Outer_2) (Variables|Default|GetCollision_Outer_2) (Variables|Default|GetVFX_Outer_2) (Variables|Default|GetAudio_Outer_2) true)
  (CallFunction|AddtoShardsMap (Variables|Default|GetShard_Outer_3) (Variables|Default|GetShardMesh_Outer_3) (Variables|Default|GetGroundMesh_Outer_3) (Variables|Default|GetCollision_Outer_3) (Variables|Default|GetVFX_Outer_3) (Variables|Default|GetAudio_Outer_3) true))

(fn SetIsRaised
  (params (Shard :type Scene%20Component%20Object%20Reference%20%28by%20ref%29) (NewIsRaised :type Boolean))
  (bind (_shardmesh_21_08daf666447f5282d423e68a2516133e _groundmesh_22_fef31656488a70c8acd70eb8edf69b97 _collision_23_1642f5a64ac4c455f2fc1ba9aec8d58c _vfx_24_40c2412946e150d9e6190593ba325161 _audio_25_22051d714a62390b3fce14bab5121281) (Utilities|Struct|BreakCorruptionPuzzleShard (Utilities|Map|Find (Variables|Data|GetShardsMap) Shard)))
  (Utilities|Map|Add (Variables|Data|GetShardsMap) Shard (Utilities|Struct|MakeCorruptionPuzzleShard _shardmesh_21_08daf666447f5282d423e68a2516133e _groundmesh_22_fef31656488a70c8acd70eb8edf69b97 _collision_23_1642f5a64ac4c455f2fc1ba9aec8d58c _vfx_24_40c2412946e150d9e6190593ba325161 _audio_25_22051d714a62390b3fce14bab5121281 NewIsRaised))
  (return))

(fn ShouldDie
  (params (Damage :type Float%20%28single-precision%29) (EventInstigator :type Controller%20Object%20Reference) (DamageCauser :type Actor%20Object%20Reference) (bIsTest :type Boolean) (EffectContext :type Gameplay%20Effect%20Context%20Handle%20Structure%20%28by%20ref%29))
  (return false))

(fn TimelineUpdate
  (params (ShardHeight :type Float%20%28double-precision%29) (ShardRotation :type Vector) (CrackScale :type Float%20%28double-precision%29) (CrackHeight :type Float%20%28double-precision%29) (Shard :type Scene%20Component%20Object%20Reference))
  (bind _shardsmap (Variables|Data|GetShardsMap))
  (bind _value (Utilities|Map|Find _shardsmap Shard))
  (bind (_shardmesh_21_08daf666447f5282d423e68a2516133e _groundmesh_22_fef31656488a70c8acd70eb8edf69b97) (Utilities|Struct|BreakCorruptionPuzzleShard _value))
  (Networking|SwitchHasAuthority
    (:Remote
      (Transformation|SetRelativeLocationAndRotation _shardmesh_21_08daf666447f5282d423e68a2516133e (Math|Vector|MakeVector 0.0 0.0 ShardHeight) (Math|Rotator|MakeRotator (.x ShardRotation) (.y ShardRotation) (.yaw (Variables|/Script/Engine.SceneComponent|GetRelativeRotation _shardmesh_21_08daf666447f5282d423e68a2516133e))) false false)
      (Transformation|SetRelativeScale3D _groundmesh_22_fef31656488a70c8acd70eb8edf69b97 (Math|Vector|MakeVector CrackScale CrackScale CrackScale))
      (Transformation|SetRelativeLocation _groundmesh_22_fef31656488a70c8acd70eb8edf69b97 (Math|Vector|MakeVector 0.0 0.0 CrackHeight) false false))))

(fn ToggleCollision
  (params (EnableCollision :type Boolean) (Shard :type Scene%20Component%20Object%20Reference%20%28by%20ref%29))
  (bind _shardsmap (Variables|Data|GetShardsMap))
  (bind _value (Utilities|Map|Find _shardsmap Shard))
  (bind (_shardmesh_21_08daf666447f5282d423e68a2516133e _groundmesh_22_fef31656488a70c8acd70eb8edf69b97 _collision_23_1642f5a64ac4c455f2fc1ba9aec8d58c) (Utilities|Struct|BreakCorruptionPuzzleShard _value))
  (if EnableCollision
    (Collision|SetCollisionEnabled _collision_23_1642f5a64ac4c455f2fc1ba9aec8d58c "QueryOnly")
    (else
      (Collision|SetCollisionEnabled _collision_23_1642f5a64ac4c455f2fc1ba9aec8d58c "NoCollision"))))

(fn ToggleCosmeticComponents
  (params (Activate :type Boolean) (Shard :type Scene%20Component%20Object%20Reference%20%28by%20ref%29))
  (bind _shardsmap (Variables|Data|GetShardsMap))
  (bind _value (Utilities|Map|Find _shardsmap Shard))
  (bind (_shardmesh_21_08daf666447f5282d423e68a2516133e _groundmesh_22_fef31656488a70c8acd70eb8edf69b97 _collision_23_1642f5a64ac4c455f2fc1ba9aec8d58c _vfx_24_40c2412946e150d9e6190593ba325161) (Utilities|Struct|BreakCorruptionPuzzleShard _value))
  (if Activate
    (Components|Activation|Activate _vfx_24_40c2412946e150d9e6190593ba325161 false)
    (else
      (Components|Activation|Deactivate _vfx_24_40c2412946e150d9e6190593ba325161))))

(fn ConstructionScript
  (|CallParentFunction)
  (Default|ResetShardsMap)
  (Variables|Data|SetShardsArray (Utilities|Array|MakeArray (Variables|Default|GetShard_Inner_1) (Variables|Default|GetShard_Inner_2) (Variables|Default|GetShard_Inner_3) (Variables|Default|GetShard_Outer_1) (Variables|Default|GetShard_Outer_2) (Variables|Default|GetShard_Outer_3))))
