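; Fortnitemares corruption puzzle shard: a quest actor, its mesh visible only to its owner, that the owner buries by damaging it.
; Burying a parent shard raises its child; an unhit child times out, reburies and re-raises the parent, while a hit child rises again after a timeout.
; When all shards are down, the owner receives the quest GameplayEffect and the shards are destroyed.
; One timeline, played forward or reversed, animates burying and raising.
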(event EventBeginPlay
  ;[1 1 1] BeginPlay - Initialize various variables
  (|CallParentFunction)
  (CallFunction|SetIsOwnerLocallyControlled self)
  (Variables|BPVars|Data|SetZ%20%28Yaw%29 (.yaw (Transformation|GetActorRotation self)))
  (Collision|SetCollisionEnabled (Variables|Default|GetCapsuleCollision) "QueryOnly")
  (if (Variables|BPVars|Data|GetIsChild)
    (CallFunction|InitializeBuried)
    (else
      (CallFunction|RaiseShard))))

(event Custom|InitializeBuried
  (bind _groundmesh (Variables|Default|GetGroundMesh))
  ;[1 1 1] Child shards start buried and don't need to be reburied cosmetically.
  (Transformation|SetRelativeLocationAndRotation (Variables|Default|GetShardMesh) "0, 0,-70.000000" (Math|Rotator|MakeRotator 0.0 0.0 (Variables|BPVars|Data|GetZ%20%28Yaw%29)) false false)
  (Transformation|SetRelativeScale3D _groundmesh "0.050000,0.050000,0.050000")
  (Transformation|SetRelativeLocation _groundmesh "0, 0,-10.000000" false false)
  (Variables|BPVars|Data|SetIsRaised false)
  (CallFunction|PauseInteraction))

(event Custom|BuryShard
  (bind _groundmesh (Variables|Default|GetGroundMesh))
  ;[1 1 1] Activate or inactivate the jewel along a timeline
  (CallFunction|PauseInteraction)
  (if (Variables|BPVars|Data|GetIsOwnerLocallyControlled)
    (if (Variables|BPVars|Data|GetIsRaised)
      (if (Variables|BPVars|Data|GetBuriedByPlayer)
        (CallFunction|ImpactFX))
      (CallFunction|SpawnFX))
    (Audio|SpawnSoundAttached (Variables|BPVars|Art|GetBurySound) (Variables|Default|GetShardMesh))
    (Components|Activation|Deactivate (Variables|Default|GetNS_QuestS18_Crystal_Ambient))
    (else
      _))
  (Variables|BPVars|Data|SetIsRaised false)
  ;;[1 1 1] Cosmetically adjust the shard and ground to raise or lower
  (bind (_direction _shardheight _shardrotation _crackscale _crackheight) (Timeline|SpinJewel PlayFromStart)
    (:Update
      (Networking|SwitchHasAuthority
        (:Remote
          (Transformation|SetRelativeLocationAndRotation (Variables|Default|GetShardMesh) (Math|Vector|MakeVector 0.0 0.0 _shardheight) (Math|Rotator|MakeRotator (.x _shardrotation) (.y _shardrotation) (Variables|BPVars|Data|GetZ%20%28Yaw%29)) false false)
          (Transformation|SetRelativeScale3D _groundmesh (Math|Vector|MakeVector _crackscale _crackscale _crackscale))
          (Transformation|SetRelativeLocation _groundmesh (Math|Vector|MakeVector 0.0 0.0 _crackheight) false false))))
    (:Finished
      (CallFunction|ResetInteraction)
      (if (Variables|BPVars|Data|GetIsRaised)
        (Utilities|FlowControl|Sequence
          (:then_0
            (Networking|SwitchHasAuthority
              (:Authority
                (CallFunction|HandleRaised))))
          (:then_1
            (if (Variables|BPVars|Data|GetIsOwnerLocallyControlled)
              (Components|Activation|Activate (Variables|Default|GetNS_QuestS18_Crystal_Ambient) false))))
        (else
          (Networking|SwitchHasAuthority
            (:Authority
              (CallFunction|HandleBuried))))))))

(event Custom|RaiseShard
  (bind _groundmesh (Variables|Default|GetGroundMesh))
  ;[1 1 1] Activate or inactivate the jewel along a timeline
  (Variables|BPVars|Data|SetIsRaised true)
  (CallFunction|PauseInteraction)
  (if (Variables|BPVars|Data|GetIsOwnerLocallyControlled)
    (CallFunction|SpawnFX)
    (Audio|SpawnSoundAttached (Variables|BPVars|Art|GetRaiseSound) (Variables|Default|GetShardMesh))
    (else
      _))
  ;;[1 1 1] Cosmetically adjust the shard and ground to raise or lower
  (Timeline|SpinJewel ReverseFromEnd))

(event Custom|SpawnFX
  ;[1 1 1] Activate or inactivate the jewel along a timeline
  (Niagara|SpawnSystematLocation "/BattlePassS18/QuestActors/Assets/FX/NS_QuestS18_Crystal_Spawn.NS_QuestS18_Crystal_Spawn" (+ (Transformation|GetActorLocation self) "0, 0,5.000000")))

(event Custom|ImpactFX
  (bind _self self)
  (bind _returnvalue (Transformation|GetActorLocation _self))
  (bind _returnvalue_2 (+ _returnvalue "0, 0,30.000000"))
  ;[1 1 1] Activate or inactivate the jewel along a timeline
  (Niagara|SpawnSystematLocation "/BattlePassS18/QuestActors/Assets/FX/NS_Quest18_Crystal_Impact.NS_Quest18_Crystal_Impact" _returnvalue_2)
  (Niagara|SpawnSystematLocation "/BattlePassS18/QuestActors/Assets/FX/NS_QuestS18_Crystal_Success.NS_QuestS18_Crystal_Success" _returnvalue_2))

(event EventOnDamageServer
  (params (Damage :type Float%20%28single-precision%29) (DamageTags :type Gameplay%20Tag%20Container%20Structure%20%28by%20ref%29) (Momentum :type Vector :default "0, 0, 0") (HitInfo :type Hit%20Result%20Structure%20%28by%20ref%29) (InstigatedBy :type Controller%20Object%20Reference) (DamageCauser :type Actor%20Object%20Reference) (EffectContext :type Gameplay%20Effect%20Context%20Handle%20Structure))
  ;[1 1 1] Activate or inactivate the jewel along a timeline
  ;;[1 1 1] On hit, we bury the shard
  (Networking|SwitchHasAuthority
    (:Authority
      (if (Utilities|Equal(Object) InstigatedBy (Fortnite|GetFortPlayerControllerfromActor (Actor|GetOwner)))
        (Variables|BPVars|Data|SetBuriedByPlayer true)
        (CallFunction|BuryShard)))))

(event Custom|HandleBuried
  (if (Variables|BPVars|Data|GetIsChild)
    (if (Variables|BPVars|Data|GetBuriedByPlayer)
      (Utilities|Time|ClearandInvalidateTimerbyHandle (Variables|BPVars|Data|GetTimer_ChildRaised))
      (Utilities|Time|ClearandInvalidateTimerbyHandle (Variables|BPVars|Data|GetTimer_ChildBuried))
      (bind _returnvalue (Utilities|Time|SetTimerbyEvent (EventDispatchers|CreateEvent :!event Custom|RaiseShard) (Variables|Default|GetShardTimeout)))
      (Variables|BPVars|Data|SetTimer_ChildBuried _returnvalue)
      (CallFunction|CheckAllBuried))
    (else
      (CallFunction|RaiseShard (Variables|BPVars|Data|GetChildShard)))))

(event Custom|CheckAllBuried
  (Variables|BPVars|Data|SetShouldApplyGE true)
  (bind _array_element (Utilities|Array|ForEachLoopwithBreak (Variables|BPVars|Data|GetAllShards))
    (:LoopBody
      (if (Variables|BPVars|Data|GetIsRaised _array_element)
        (Variables|BPVars|Data|SetShouldApplyGE false)))
    (:Completed
      _))
  (if (Variables|BPVars|Data|GetShouldApplyGE)
    (Environment|ApplyGlobalEnvironmentGameplayEffecttoActor (Actor|GetOwner) (Variables|BPVars|Data|GetQuestGE))
    (for _array_element_2 (Variables|BPVars|Data|GetAllShards)
      (Utilities|Time|ClearandInvalidateTimerbyHandle (Variables|BPVars|Data|GetTimer_ChildBuried _array_element_2))
      (Utilities|Time|ClearandInvalidateTimerbyHandle (Variables|BPVars|Data|GetTimer_ChildRaised _array_element_2))
      (Actor|DestroyActor _array_element_2))))

(event Custom|HandleRaised
  (if (Variables|BPVars|Data|GetIsChild)
    (Utilities|Time|ClearandInvalidateTimerbyHandle (Variables|BPVars|Data|GetTimer_ChildRaised))
    (Utilities|Time|ClearandInvalidateTimerbyHandle (Variables|BPVars|Data|GetTimer_ChildBuried))
    (bind _returnvalue (Utilities|Time|SetTimerbyEvent (EventDispatchers|CreateEvent :!event Custom|ReburyChildRaiseParent) (Variables|Default|GetShardTimeout)))
    (Variables|BPVars|Data|SetTimer_ChildRaised _returnvalue)))

(event Custom|ReburyChildRaiseParent
  (Variables|BPVars|Data|SetBuriedByPlayer false)
  (CallFunction|BuryShard)
  (CallFunction|RaiseShard (Variables|BPVars|Data|GetParentShard)))

(event Custom|ProgressCosmetics
  ;[0.03368 1 0] Progress made cosmetics!
  (bind _returnvalue (Niagara|SpawnSystemAttached "/Game/Environments/Sets/ArtDeco/Props/FX/NS_Gems_Spawn.NS_Gems_Spawn" (Variables|Default|GetShardMesh) "None" "0,0.500000,2.500000"))
  (Rendering|SetOnlyOwnerSee _returnvalue true :!class /Script/Engine.PrimitiveComponent)
  (if (Variables|BPVars|Data|GetIsOwnerLocallyControlled)
    (Audio|SpawnSoundAttached (Variables|BPVars|Art|GetProgressSound) (Variables|Default|GetShardMesh))))

(fn OnRep_IsChild)

(fn PauseInteraction
  (Collision|SetCollisionEnabled (Variables|Default|GetCapsuleCollision) "NoCollision")
  (return))

(fn ResetInteraction
  (if (Variables|BPVars|Data|GetIsRaised)
    (Collision|SetCollisionEnabled (Variables|Default|GetCapsuleCollision) "QueryOnly")
    (return)))

(fn SetIsOwnerLocallyControlled
  (params (ActorToCheck :type Actor%20Object%20Reference))
  (bind _array_element (Utilities|Array|ForEachLoopwithBreak (Fortnite|GetLocalFortPlayerControllers true))
    (:LoopBody
      (if (Utilities|Equal(Object) _array_element (Fortnite|GetFortPlayerControllerfromActor (Actor|GetOwner ActorToCheck)))
        (Variables|BPVars|Data|SetIsOwnerLocallyControlled true)))
    (:Completed
      _))
  (return))

(fn ShouldDie
  (params (Damage :type Float%20%28single-precision%29) (EventInstigator :type Controller%20Object%20Reference) (DamageCauser :type Actor%20Object%20Reference) (bIsTest :type Boolean) (EffectContext :type Gameplay%20Effect%20Context%20Handle%20Structure%20%28by%20ref%29))
  (return false))

(fn ConstructionScript
  (|CallParentFunction)
  (Rendering|SetOnlyOwnerSee (Variables|Default|GetShardMesh) true :!class /Script/Engine.PrimitiveComponent))
