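; Quest actor for the Fortnitemares corruption minigame that spawns puzzle shards when a player enters its collision volume.
; On authority, overlap spawns shards once per pawn if the quest is in progress or Test is set.
; Six shards spawn at SpawnCollision's child transforms, linked as three parent-child pairs, each receiving the full set.
; The actor destroys itself on BeginPlay if AthenaCheckCreativeMode returns false, and overlap repeats that check.
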(event EventBeginPlay
  (bind _cancompletequest (Class|AthenaQuestFunctionLibrary|AthenaCheckCreativeMode (Game|GetGameState)))
  (if (not _cancompletequest)
    (Actor|DestroyActor)))

(event OnComponentBeginOverlap(SpawnCollision)
  (params (OverlappedComponent :type Primitive%20Component%20Object%20Reference) (OtherActor :type Actor%20Object%20Reference) (OtherComp :type Primitive%20Component%20Object%20Reference) (OtherBodyIndex :type Integer) (bFromSweep :type Boolean) (SweepResult :type Hit%20Result%20Structure%20%28by%20ref%29))
  (Networking|SwitchHasAuthority
    (:Authority
      (if (== (Utilities|Array|FindItem (Variables|Default|GetStoredPawns) OtherActor) -1)
        (bind _asfort_player_pawn (Utilities|Casting|CastToFortPlayerPawn OtherActor))
        (Utilities|IsValid (Fortnite|GetFortPlayerControllerfromActor _asfort_player_pawn)
          (:Is%20Valid
            (Utilities|IsValid (Quests|GetQuestManager (Fortnite|GetFortPlayerControllerfromActor _asfort_player_pawn) "Invalid")
              (:Is%20Valid
                (bind _cancompletequest (Class|AthenaQuestFunctionLibrary|AthenaCheckCreativeMode (Game|GetGameState)))
                (if _cancompletequest
                  (if (or (Quest|IsQuestinProgress (Quests|GetQuestManager (Fortnite|GetFortPlayerControllerfromActor _asfort_player_pawn) "Invalid") (Variables|Default|GetQuestID)) (Variables|Default|GetTest))
                    (Utilities|Array|AddUnique (Variables|Default|GetStoredPawns) _asfort_player_pawn)
                    (CallFunction|SpawnJewels _asfort_player_pawn)))))))))))

(fn GetLocationOffset
  (params (Index :type Integer))
  (return (Transformation|GetWorldTransform (Utilities|Array|Get(acopy) (Components|GetChildrenComponents (Variables|Components|GetSpawnCollision) false) Index))))

(fn SpawnJewels
  (params (Instigator :type Pawn%20Object%20Reference))
  (bind _shards (LocalVariables|GetShards))
  (bind _shards_2 (LocalVariables|GetShards))
  (for _index (range 6)
    (bind _returnvalue (Game|SpawnActorfromClass "/BattlePassS18/QuestActors/Fortnitemares/CorruptionMinigame/BP_Quest_CorruptionPuzzle_Shard.BP_Quest_CorruptionPuzzle_Shard_C" (CallFunction|GetLocationOffset _index) "Undefined" "OverrideRootScale" Instigator true (select (< _index 3) false true) 0 Instigator))
    (Utilities|Array|Add (LocalVariables|GetShards) _returnvalue))
  (if (Utilities|IsArrayNonEmpty _shards_2)
    (for (_array_element _array_index) _shards_2
      (if (Math|Boolean|ORBoolean (== _array_index 0) (== _array_index 1) (== _array_index 2))
        (Variables|/BattlePassS18/QuestActors/Fortnitemares/CorruptionMinigame/BP_Quest_CorruptionPuzzle_Shard.BP_Quest_CorruptionPuzzle_Shard_C|SetChildShard (Utilities|Array|Get(acopy) _shards (+ _array_index 3)) _array_element)
        (else
          (Variables|/BattlePassS18/QuestActors/Fortnitemares/CorruptionMinigame/BP_Quest_CorruptionPuzzle_Shard.BP_Quest_CorruptionPuzzle_Shard_C|SetParentShard (Utilities|Array|Get(acopy) _shards (- _array_index 3)) _array_element)))
      (Variables|/BattlePassS18/QuestActors/Fortnitemares/CorruptionMinigame/BP_Quest_CorruptionPuzzle_Shard.BP_Quest_CorruptionPuzzle_Shard_C|SetAllShards (LocalVariables|GetShards) _array_element))))

(fn ConstructionScript)
